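# Translucent unlit, two-sided material for Lightning Dragon backpack lightning, used on Niagara ribbons and mesh particles.
# Dynamic parameters drive texture UV offset and scale, dissolve, and appear/disappear wipes, which combine with edge masks, view-angle fade and depth fade for opacity.
# Emissive blends two colors by particle color, plus an optional curve-atlas lead-edge glow.
# World position offset adds time-animated sine jitter scaled per particle, plus optional SplineThicken width.
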Material /Game/Accessories/FORT_Backpacks/Backpack_F_MED_LightningDragon/FX/M_Backpack_LightningDragon_Lightning
graph {
 properties {
  bUsedWithNiagaraMeshParticles=true
  bUsedWithNiagaraRibbons=true
  blendMode=BLEND_Translucent
  pixelDepthOffsetMode=PDOM_Legacy
  shadingModel=MSM_Unlit
  twoSided=true
  usedShadingModels=Unlit
 }
 nodes [
  n1 = LinearInterpolate(Alpha=DynamicParameter() {parameterIndex=2}.Param1) {constA=-0.7}
  Divide(A=@n1)
  n2 = ParticleColor()
  n3 = ScalarParameter() {defaultValue=1, parameterName=U offset}
  n4 = DynamicParameter() {parameterIndex=1, defaultValue=<rgba 1 1 2 1>, paramNames [Width Scale, U offset, U scale, Dissappear]}
  n5 = ScalarParameter() {parameterName=Use Dy UV setting}
  n6 = Add(A=TextureCoordinate(), B=AppendVector(A=LinearInterpolate(A=@n3, B=@n4.U offset, Alpha=@n5), B=Constant()))
  n7 = ScalarParameter() {defaultValue=2, parameterName=U Scale}
  n8 = DynamicParameter() {parameterIndex=1, defaultValue=<rgba 1 1 2 1>, paramNames [Width Scale, U offset, U scale, Dissappear]}
  n9 = ScalarParameter() {parameterName=Use Dy UV setting}
  n10 = VertexColor()
  n11 = AppendVector(
   A=Multiply(A=LinearInterpolate(A=@n7, B=@n8.U scale, Alpha=@n9), B=LinearInterpolate(Alpha=@n10.R) {constA=1})
   B=ScalarParameter() {defaultValue=-1, parameterName=V Scale}
  )
  n12 = Constant2Vector() {r=0.5}
  n13 = LinearInterpolate(Alpha=VertexColor().R) {constA=1}
  n14 = TextureSampleParameter2D(UVs=Add(A=Multiply(A=@n6, B=@n11), B=Multiply(A=@n12, B=@n13))) {parameterName=Lightning, texture=&/T_Backpack_LightningDragon_Lightning}
  n15 = Multiply(
   A=Multiply(A=@n2.A, B=StaticComponentMaskParameter(None=@n14.RGB) {defaultR=true, parameterName=TexRGBMask})
   B=MaterialFunctionCall("FadeSensitivity (optional)"=ScalarParameter() {defaultValue=1, parameterName=FadeBasedOnViewAngle}) {materialFunction=&FadeBasedOnViewAngle}.Result
  )
  n16 = DynamicParameter() {defaultValue=<rgba 1 1 0 0>, paramNames [Appear,U_scale,dissolve,WPO]}
  n17 = Multiply(A=LinearInterpolate(A=@n15, B=Subtract(A=@n15, B=@n16.dissolve), Alpha=ScalarParameter() {defaultValue=1, parameterName=Use Dissolve}) {constA=1})
  n18 = TextureCoordinate() {uTiling=0}
  n19 = Constant2Vector() {g=0.5}
  n20 = Saturate(None=Multiply(A=SphereMask(A=@n18, B=@n19) {attenuationRadius=0.5, hardnessPercent=0}) {constB=3})
  n21 = ComponentMask(None=TextureCoordinate()) {r=true}
  n22 = LinearInterpolate(A=OneMinus(None=@n21), B=@n21, Alpha=ScalarParameter() {parameterName=Use the opposite mask})
  n23 = DynamicParameter() {defaultValue=<rgba 1 1 0 0>, paramNames [Appear,U_scale,dissolve,WPO]}
  n24 = Saturate(None=Add(A=LinearInterpolate(Alpha=@n22) {constA=-0.5}, B=LinearInterpolate(Alpha=@n23.Appear) {constA=-1, constB=2}))
  n25 = LinearInterpolate(Alpha=ComponentMask(None=TextureCoordinate()) {r=true}) {constA=-0.5}
  n26 = DynamicParameter() {parameterIndex=1, defaultValue=<rgba 1 1 2 1>, paramNames [Width Scale, U offset, U scale, Dissappear]}
  n27 = LinearInterpolate(Alpha=@n26.Dissappear) {constA=-1, constB=2}
  n28 = LinearInterpolate(B=Saturate(None=Add(A=@n25, B=@n27)), Alpha=ScalarParameter() {parameterName=Use disappear end}) {constA=1}
  n29 = Multiply(
   A=Multiply(A=Multiply(A=@n17, B=@n20), B=Multiply(A=@n24, B=@n28))
   B=LinearInterpolate(
    B=TextureSample() {texture=/Game/Weapons/FORT_Melee/Pickaxe_LightningDragon/FX/T_Edge_Mask}.R
    Alpha=ScalarParameter() {parameterName=Use edge mask}
   ) {constA=1}
  )
  n30 = Sine(None=ComponentMask(None=TextureCoordinate()) {r=true}) {period=2}
  n31 = ScalarParameter() {defaultValue=1, parameterName=Edge mask02 power}
  n32 = Saturate(None=Multiply(A=Power(Base=@n30, Exp=@n31)) {constB=2})
  n33 = ScalarParameter() {defaultValue=1, parameterName=use edge mask02}
  n34 = ComponentMask(None=TextureCoordinate()) {r=true}
  n35 = Saturate(None=Multiply(A=Power(Base=@n34)) {constB=5})
  n36 = ScalarParameter() {parameterName=Use lead mask}
  n37 = Multiply(A=@n29, B=Multiply(A=LinearInterpolate(B=@n32, Alpha=@n33) {constA=1}, B=LinearInterpolate(B=@n35, Alpha=@n36) {constA=1}))
  n38 = DepthFade(Opacity=@n37, FadeDistance=ScalarParameter() {defaultValue=5, parameterName=FadeDistance}) {fadeDistanceDefault=5}
  LinearInterpolate(
   A=ShadingPathSwitch(Default=@n38, Deferred=@n38, Forward=@n37, Mobile=@n37)
   B=Multiply(A=NamedRerouteUsage(), B=@n29)
   Alpha=ScalarParameter() {group=LeadEdge, parameterName=UseLeadEdge}
  )
  n39 = Power(Base=SphereMask(A=TextureCoordinate() {vTiling=0}, B=Constant2Vector() {r=0.5}) {attenuationRadius=0.5, hardnessPercent=5}) {constExponent=5}
  n40 = Multiply(
   A=LinearInterpolate(
    A=VectorParameter() {parameterName=Color01, defaultValue=<rgba 1 0.4325 0.0214 0>}.RGB
    B=VectorParameter() {parameterName=Color02, defaultValue=<rgba 1 0.7652 0 0>}.RGB
    Alpha=Saturate(None=@n39)
   )
   B=@n2.RGB
  )
  n41 = Multiply(A=@n40, B=Multiply(A=NamedRerouteUsage(), B=ScalarParameter() {defaultValue=100, group=LeadEdge, parameterName=LeadEdgeEmissive}))
  MaterialFunctionCall(Emissive=LinearInterpolate(A=@n40, B=Add(A=@n41, B=@n40), Alpha=ScalarParameter() {group=LeadEdge, parameterName=UseLeadEdge})) {materialFunction=&MF_FX_EmissiveAdjustments}
  MaterialFunctionCall() {materialFunction=&DebugTimeSine}
  n42 = TextureCoordinate()
  n43 = Multiply(A=Sine(None=ComponentMask(None=@n42) {g=true}) {period=2}) {constB=0.5}
  n44 = TextureSampleParameter2D(UVs=Add(A=TextureCoordinate() {uTiling=2, vTiling=2}, B=Constant2Vector())) {parameterName=Dissolve Tex, texture=/Game/Effects/Fort_Effects/Textures/T_Tile_Pattern_01}
  MaterialFunctionCall(
   Transition Phase=MaterialFunctionCall() {materialFunction=&DebugTimeSine}.Result
   Height Texture=Saturate(None=Add(A=@n43, B=@n44.R))
  ) {materialFunction=&HeightLerp}
  MaterialFunctionCall() {materialFunction=&DebugTimeSine}
  n45 = Divide(
   A=MaterialFunctionCall() {materialFunction=&LinearGradient}.UGradient
   B=ScalarParameter() {defaultValue=2, group=LeadEdge, parameterName=LeadEdgeScalar}
  )
  n46 = CurveAtlasRowParameter(CurveTime=Add(A=OneMinus(None=@n45), B=@n1)) {
   atlas=/Game/Athena/LinearColorCurves/CA_Contrail_SilentTempo
   curve=/Game/Accessories/FORT_Backpacks/Backpack_F_MED_FishBowlBone/FX/Curve_LeadEdge
   defaultValue=14
   group=LeadEdge
   parameterName=LeadEdgeCurve
  }
  NamedRerouteDeclaration(None=Multiply(A=@n46.R)) {name=LeadEdge, nodeColor=<rgba 0.6875 1 0 1>}
  NamedRerouteDeclaration(None=Add(A=ScalarParameter() {parameterName=TimeOffset}, B=Time())) {name=Time, nodeColor=<rgba 1 0 0.1118 1>}
  n47 = ScalarParameter() {defaultValue=-4, group=WPO, parameterName=WPO_Speed}
  n48 = Multiply(A=NamedRerouteUsage(), B=@n47) {constB=2}
  n49 = Multiply(A=ComponentMask(None=TextureCoordinate()) {r=true}, B=ScalarParameter() {defaultValue=4, group=WPO, parameterName=WPO_U_Tile})
  n50 = Ceil(None=Sine(None=Add(A=@n49, B=@n48)))
  n51 = Constant3Vector() {constant=<rgba 1 0 0 0>}
  n52 = Multiply(A=Transform(None=Multiply(A=@n50, B=@n51)) {transformSourceType=TRANSFORMSOURCE_Local})
  n53 = DynamicParameter() {defaultValue=<rgba 1 1 0 0>, paramNames [Appear,U_scale,dissolve,WPO]}
  n54 = NamedRerouteUsage()
  n55 = Sine(None=Add(A=@n49, B=Multiply(A=@n54, B=@n47) {constB=2})) {period=1.5}
  n56 = Transform(None=Multiply(A=Saturate(None=@n55), B=Constant3Vector() {constant=<rgba 0 1 0 0>})) {transformSourceType=TRANSFORMSOURCE_Local}
  n57 = Multiply(A=Add(A=Multiply(A=@n52, B=@n53.WPO), B=Multiply(A=@n56, B=@n53.WPO)))
  n58 = DynamicParameter() {parameterIndex=1, defaultValue=<rgba 1 1 2 1>, paramNames [Width Scale, U offset, U scale, Dissappear]}
  n59 = Multiply(A=@n58.Width Scale, B=ScalarParameter() {defaultValue=3, parameterName=width})
  n60 = Sine(None=ComponentMask(None=TextureCoordinate()) {r=true}) {period=2}
  n61 = LinearInterpolate(B=Saturate(None=Multiply(A=@n60) {constB=2}), Alpha=ScalarParameter() {parameterName=UseWidthMask}) {constA=1}
  n62 = ScalarParameter() {defaultValue=1, parameterName=SplineTipScale}
  n63 = MaterialFunctionCall(
   WidthBase=Multiply(A=@n59, B=@n61)
   WidthTip=Multiply(A=Multiply(A=@n59, B=@n62), B=@n61)
   Expand U or V UV Channel=StaticBoolParameter() {parameterName="Expand U or V UV Channel (B)"}
  ) {materialFunction=&SplineThicken}
  n64 = ScalarParameter() {parameterName=Use spline thicken}
  Reroute(None=Add(A=Multiply(A=@n57), B=LinearInterpolate(B=@n63.WorldPositionOffset, Alpha=@n64)))
 ]
}
